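# Animation Blueprint for the DorsalDance pickaxe's offhand axe, animating its jaw and fins.
# Spring-driven Anim Dynamics on the jaw feeds additive jaw and fin bone modifiers, then an additive state machine.
# That machine switches between looping Idle and Closed pose while the owner's Fists harvesting montage plays.
# The event graph sets that swing flag, a sine-based jaw rotation, curve-driven idle alpha and delayed-reset melee alpha.
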AnimBlueprint /BRCosmetics/Weapons/FORT_Melee/Pickaxe_DorsalDance/Meshes/DorsalDance_Axe_Offhand_AnimBP
target_skeleton=&/DorsalDance_Axe_Skeleton
graphs [
 AnimGraph {nodes [
   n1 = Anim Dynamics - Bone: dyn_jaw_c(ComponentPose=Local To Component(LocalPose=Input Pose()), Alpha=1, ExternalForce="0,0,0") {
    angularDampingOverride=1
    angularSpringConstant=20000
    bAngularSpring=true
    bOverrideAngularDamping=true
    bUseGravityOverride=true
    gravityScale=0.35
    simulationSpace=RootRelative
    alphaBoolBlend {blendOption=Linear}
    boundBone {boneName=dyn_jaw_c}
    componentLinearAccScale=<1 1 1>
    componentLinearVelScale=<1 1 1>
    gravityOverride {z=-980}
    physicsBodyDefinitions [{
      boundBone {boneName=dyn_jaw_c}
      boxExtents=<20 30 30>
      constraintSetup {angularLimitsMax=<0 5 0>, angularLimitsMin=<0 -50 0>, angularTarget=<1 0 0>}
      localJointOffset=<4 0 0>
     }]
   }
   n2 = "Transform (Modify) Bone - Bone: dyn_jaw_c"(ComponentPose=@n1, Alpha=1, Translation="0,0,0", Rotation="0,0,0", Scale="1,1,1") {
    rotationMode=BMM_Additive
    translationMode=BMM_Additive
    alphaBoolBlend {blendOption=Linear}
    alphaScaleBiasClamp {bInterpResult=true, interpSpeedDecreasing=3}
    boneToModify {boneName=dyn_jaw_c}
   }
   n3 = "Transform (Modify) Bone - Bone: dyn_jaw_c"(ComponentPose=@n2, Alpha=1, Translation="0,0,0", Rotation="0,0,0", Scale="1,1,1") {
    rotationMode=BMM_Additive
    translationMode=BMM_Additive
    alphaBoolBlend {blendOption=Linear}
    alphaScaleBiasClamp {bInterpResult=true, interpSpeedDecreasing=3}
    boneToModify {boneName=dyn_jaw_c}
   }
   n4 = "Transform (Modify) Bone - Bone: dyn_fin_l"(ComponentPose=@n3, Alpha=1, Translation="0,0,0", Rotation="0,0,0", Scale="1,1,1") {
    rotationMode=BMM_Additive
    translationMode=BMM_Additive
    alphaBoolBlend {blendOption=Linear}
    alphaScaleBiasClamp {bInterpResult=true, interpSpeedDecreasing=3}
    boneToModify {boneName=dyn_fin_l}
   }
   n5 = "Transform (Modify) Bone - Bone: dyn_fin_r"(ComponentPose=@n4, Alpha=1, Translation="0,0,0", Rotation="0,0,0", Scale="1,1,1") {
    rotationMode=BMM_Additive
    translationMode=BMM_Additive
    alphaBoolBlend {blendOption=Linear}
    alphaScaleBiasClamp {bInterpResult=true, interpSpeedDecreasing=3}
    boneToModify {boneName=dyn_fin_r}
   }
   AnimGraphNode_StateMachine_0 = Idle and Pose() {is=title AnimGraphNode_StateMachine}
   Output Pose(Result=Apply Additive(Base=Component To Local(ComponentPose=@n5), Additive=@AnimGraphNode_StateMachine_0) {alphaBoolBlend {blendOption=Linear}})
  ]}
 Idle and Pose {
  entry=Idle
  path=AnimGraph.AnimGraphNode_StateMachine_0.Idle and Pose
  ids=id
  nodes [
   AnimStateNode_0 = Idle()
   AnimStateNode_1 = Closed() {is=title AnimStateNode}
   AnimStateTransitionNode_0 = AnimStateTransitionNode()
   AnimStateTransitionNode_1 = AnimStateTransitionNode()
  ]
  links [
   {
    dst=Idle
    rule=AnimGraph.AnimGraphNode_StateMachine_0.Idle and Pose.AnimStateTransitionNode_1.Transition
    src=Closed
    via=AnimStateTransitionNode_1
   }
   {
    dst=Closed
    rule=AnimGraph.AnimGraphNode_StateMachine_0.Idle and Pose.AnimStateTransitionNode_0.Transition
    src=Idle
    via=AnimStateTransitionNode_0
   }
  ]
 }
 Idle {
  path=AnimGraph.AnimGraphNode_StateMachine_0.Idle and Pose.AnimStateNode_0.Idle
  nodes [Output Animation Pose(Result="Sequence Player (additive) 'DorsalDance_Idle_Anim_WEP'"() {sequence=/BRCosmetics/Weapons/FORT_Melee/Pickaxe_DorsalDance/Animation/DorsalDance_Idle_Anim_WEP, startPosition=2})]
 }
 Closed {
  path=AnimGraph.AnimGraphNode_StateMachine_0.Idle and Pose.AnimStateNode_1.Closed
  nodes [Output Animation Pose(Result="Sequence Player (additive) 'DorsalDance_Base_Pose_WEP'"() {bLoopAnimation=false, sequence=/BRCosmetics/Weapons/FORT_Melee/Pickaxe_DorsalDance/Animation/DorsalDance_Base_Pose_WEP})]
 }
 Transition {
  path=AnimGraph.AnimGraphNode_StateMachine_0.Idle and Pose.AnimStateTransitionNode_0.Transition
  nodes [Result(bCanEnterTransition=Get Is Swinging() -> (Is Swinging:bool))]
 }
 Transition {
  path=AnimGraph.AnimGraphNode_StateMachine_0.Idle and Pose.AnimStateTransitionNode_1.Transition
  nodes [Result(bCanEnterTransition=NOT Boolean(A=Get Is Swinging()))]
 }
 EventGraph {nodes [
   Event BlueprintUpdateAnimation()
   > n1 = Sequence!(execute) -> (then_0, then_1, then_2:Exec)
   n2 = GetRealTimeSeconds()
   n3 = "Sin (Radians)"(A=float * float(A=@n2, B=float * float(A=0.5, B=2)))
   n4 = "Sin (Radians)"(A=float * float(A=float * float(A=2, B=20), B=@n2))
   > Set JawRotation(JawRotation:Rotator=MakeRotator(Pitch=Remap Float(Input=@n4, OldRange=(X=-1.5,Y=1.5), NewRange=(X=-2,Y=30)) -> (Output=~), Yaw=float * float(A=@n3, B=0))) -> (Output_Get:Rotator)
   > n5 = Branch(Condition=Get bIsPlayingMeleeAnim())
   > Set MeleeAttackAlpha(MeleeAttackAlpha:double=1) -> (Output_Get:double)
   Delay(execute=@n5.else, Duration=0.5)
   > Set MeleeAttackAlpha(MeleeAttackAlpha=0)
   Set Is Swinging(
    execute=@n1.then_2
    Is Swinging:bool=Montage_IsPlaying(
     self=GetOwnerFortPlayerAnimInstance!(self:Custom Character Part Anim Instance Object Reference) -> (ReturnValue:Fort Player Anim Instance Object Reference)
     Montage=/Game/Animation/Game/MainPlayer/Combat/Melee/DualWield/Fists/Fists_Harvesting_Combo_M
    )
   ) -> (Output_Get:bool)
   Set IdleAlpha(execute=@n1.then_1, IdleAlpha:double=GetCurveValue(CurveName=IdleCurve)) -> (Output_Get:double)
   TryGetPawnOwner()
   n6 = float * float(A=1.1, B=1.1)
   float * float(A="Sin (Radians)"(A=float * float(A=@n6, B=@n2)), B=0.3)
  ]}
]
variables [
 IdleAlpha {type=number}
 Is Swinging {type=boolean}
 JawRotation {type=Rotator}
 MeleeAttackAlpha {type=number}
]
